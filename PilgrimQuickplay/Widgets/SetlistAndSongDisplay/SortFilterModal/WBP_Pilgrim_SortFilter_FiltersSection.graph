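# Labeled section of the Pilgrim sort/filter modal that shows filter options as toggle buttons in two columns.
# SetupItems clears old entries, then creates a button per item viewmodel, alternating columns by index, and adds it to a selection-required button group.
# The active item's button is selected silently, and clicks are re-broadcast through OnItemClicked with the item viewmodel.
# Resets unbind click handlers but keep widgets for reuse; mobile gets wider entry spacing.
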WidgetBlueprint /PilgrimQuickplay/Widgets/SetlistAndSongDisplay/SortFilterModal/WBP_Pilgrim_SortFilter_FiltersSection
graphs [
 DetermineColumnToShow {nodes [
   n1 = DetermineColumnToShow!() -> (then, Index:int) {is=title K2Node_FunctionEntry}
   n2 = "% (Integer)"(A=@n1, B=2) -> (ReturnValue=0)
   > Return Node!(
    execute
    EntryColumnReference:Dynamic Entry Box Object Reference=Select(
     Option 0:Dynamic Entry Box Object Reference=Get DynBox_Entries_Col2() -> (DynBox_Entries_Col2:Dynamic Entry Box Object Reference)
     Option 1:Dynamic Entry Box Object Reference=Get DynBox_Entries_Col1() -> (DynBox_Entries_Col1:Dynamic Entry Box Object Reference)
     Index="Equal (Integer)"(A=@n2)
    ) -> (ReturnValue:Dynamic Entry Box Object Reference)
   )
  ]}
 EventGraph {nodes [
   n1 = Event PreConstruct()
   > "SetText (Text)"(
    self=Get Text_SectionName() -> (Text_SectionName:Common Text Object Reference)
    InText=Select(Option 0:Text=Get SectionLabel() -> (SectionLabel:Text), Option 1:Text="INVTEXT(\"[PH] Section Text\")", Index=@n1.IsDesignTime) -> (ReturnValue:Text)
   )
   > SetMobileStyle() {is=title K2Node_CallFunction}
   Event OnInitialized()
   > n2 = Construct Object from Class!(execute, Class=/Script/CommonUI.CommonButtonGroupBase, self, bSelectionRequired:bool=true) -> (then, ReturnValue:Common Button Group Base Object Reference)
   > Set ButtonGroup(ButtonGroup:Common Button Group Base Object Reference=@n2) -> (Output_Get:Common Button Group Base Object Reference)
  ]}
 HandleButtonClicked {nodes [
   n1 = HandleButtonClicked!() -> (then, Item VM:Fort Sort Filter Item VM Object Reference) {is=title K2Node_FunctionEntry}
   > Call On Item Clicked!(execute:Exec, self:WBP Pilgrim Sort Filter Filters Section Object Reference, Item Filter VM:Fort Sort Filter Item VM Object Reference=@n1) -> (then:Exec) {is=title K2Node_CallDelegate}
  ]}
 OnItemClicked {nodes [OnItemClicked!() -> (then, Item Filter VM:Fort Sort Filter Item VM Object Reference) {is=title K2Node_FunctionEntry}]}
 ResetEntries {nodes [
   ResetEntries() {is=title K2Node_FunctionEntry}
   > RemoveAll(self:Common Widget Group Base Object Reference=Get ButtonGroup() -> (ButtonGroup:Common Button Group Base Object Reference))
   > n1 = Set CompleteList(CompleteList:Array of User Widget Object References=GetAllEntries!(self:Dynamic Entry Box Base Object Reference=Get DynBox_Entries_Col1()) -> (ReturnValue:Array of User Widget Object References)) -> (Output_Get:Array of User Widget Object References)
   > Append Array!(
    execute
    TargetArray:Array of User Widget Object References=@n1
    SourceArray:Array of User Widget Object References=GetAllEntries(self=Get DynBox_Entries_Col2())
   ) -> (then)
   > n2 = For Each Loop(Array:Array of User Widget Object References=Get CompleteList() -> (CompleteList:Array of User Widget Object References)) -> (Array Element:User Widget Object Reference)
   Reset!(execute=@n2.Completed, self:Dynamic Entry Box Object Reference=Get DynBox_Entries_Col1(), bDeleteWidgets:bool=false) -> (then)
   > Reset(self=Get DynBox_Entries_Col2(), bDeleteWidgets=false)
   n3 = Cast To WBP_Pilgrim_SortFilterButton!(execute=@n2.LoopBody, Object=@n2) -> (then, CastFailed, AsWBP Pilgrim Sort Filter Button:WBP Pilgrim Sort Filter Button Object Reference) {is=title K2Node_DynamicCast}
   > Unbind Event from On Button Clicked(self:WBP Pilgrim Sort Filter Button Object Reference=@n3, Delegate=Create Event()) {is=title K2Node_RemoveDelegate}
  ]}
 SetMobileStyle {nodes [
   SetMobileStyle() {is=title K2Node_FunctionEntry}
   > Branch(Condition=IsMobileGame())
   > SetEntrySpacing!(execute, self:Dynamic Entry Box Base Object Reference=Get DynBox_Entries_Col1(), InEntrySpacing_X:double=0, InEntrySpacing_Y:double=32) -> (then)
   > SetEntrySpacing(self=Get DynBox_Entries_Col2(), InEntrySpacing_X=0, InEntrySpacing_Y=32)
  ]}
 SetupItems {nodes [
   n1 = SetupItems!() -> (then, Array:Array of MVVM Base Viewmodel Object References) {is=title K2Node_FunctionEntry}
   > ResetEntries() {is=title K2Node_CallFunction}
   > n2 = For Each Loop(Array:Array of MVVM Base Viewmodel Object References=@n1) -> (Array Element:MVVM Base Viewmodel Object Reference)
   SelectButtonAtIndex!(
    execute=@n2.Completed
    self:Common Button Group Base Object Reference=Get ButtonGroup()
    ButtonIndex:int=Get ActiveButtonIndex() -> (ActiveButtonIndex:int=0)
    bAllowSound:bool=false
   ) -> (then)
   n3 = Cast To FortSortFilterItemVM!(execute=@n2.LoopBody, Object=@n2) -> (then, CastFailed, AsFort Sort Filter Item VM:Fort Sort Filter Item VM Object Reference) {is=title K2Node_DynamicCast}
   > n4 = Sequence()
   > Branch(Condition=Get bIsActive!(self:Fort Sort Filter Item VM Object Reference=@n3) -> (bIsActive:bool))
   > Set ActiveButtonIndex(ActiveButtonIndex:int=@n2.Array Index) -> (Output_Get:int=0)
   n5 = DetermineColumnToShow!(execute=@n4.then_1, self, Index:int=@n2.Array Index) -> (then, EntryColumnReference:Dynamic Entry Box Object Reference) {is=title K2Node_CallFunction}
   > n6 = Create Entry!(execute, self:Dynamic Entry Box Object Reference=@n5) -> (then, ReturnValue:User Widget Object Reference)
   > n7 = Cast To WBP_Pilgrim_SortFilterButton(Object=@n6)
   > AddWidget!(
    execute
    self:Common Widget Group Base Object Reference=Get ButtonGroup()
    InWidget:Widget Object Reference=Get WBP_UIKit_ButtonWithToggle!(self:WBP Pilgrim Sort Filter Button Object Reference=@n7) -> (WBP_UIKit_ButtonWithToggle:WBP UIKit Button With Toggle Object Reference)
   ) -> (then)
   > SetFortSortFilterItemVM!(execute, self:WBP Pilgrim Sort Filter Button Object Reference=@n7, Viewmodel:Fort Sort Filter Item VM Object Reference=@n3) -> (then)
   > Bind Event to On Button Clicked(self:WBP Pilgrim Sort Filter Button Object Reference=@n7, Delegate=Create Event()) {is=title K2Node_AddDelegate}
  ]}
]
variables [
 Button {type=CommonButtonBase, default=null}
 ButtonGroup {type=CommonButtonGroupBase, default=null}
 CompleteList {type="array<UserWidget>", default []}
 Dyn Box Entries Col 2 {type=DynamicEntryBox, default=null}
 OnItemClicked {}
 SectionLabel {type=string, default=""}
 SectionString {type=string, default=""}
]
widgets {
 info {inheritedWidgetCount=0, parentClass=&UserWidget, rootWidgetClass=&VerticalBox, widgetCount=6}
 widgets [
  VerticalBox_0 {widgetClassPath=&VerticalBox, widget=VerticalBox_0}
  Text_SectionName {
   widgetClassPath=&CommonTextBlock
   bIsVariable=true
   parent=VerticalBox_0
   widget=Text_SectionName
   properties {
    bAutoCollapseWithEmptyText=true
    clipping=OnDemand
    scrollStyle=/FortUILibrary/Styles/Typography/Foundation/TextScrollStyle-NoFade.TextScrollStyle-NoFade_C
    text=Section
    textOverflowPolicy=Ellipsis
    font {
     bForceMonospaced=false
     bMaterialIsStencil=false
     fontMaterial=null
     fontObject=&HeadingNow
     letterSpacing=0
     monospacedWidth=1
     size=20.25
     skewAmount=0
     typefaceFontName=75 Medium
     outlineSettings {
      bApplyOutlineToDropShadows=false
      bMiteredCorners=false
      bSeparateFillAlpha=false
      outlineMaterial=null
      outlineSize=0
      outlineColor=<rgba 0 0 0 1>
     }
    }
    shadowOffset=<0 0>
   }
   slot {class=VerticalBoxSlot, properties {padding=<ltrb 10 0 10 0>}}
  }
  HorizontalBox_0 {
   widgetClassPath=&HorizontalBox
   parent=VerticalBox_0
   widget=HorizontalBox_0
   slot {class=VerticalBoxSlot, properties {padding=<ltrb 10 20 10 20>}}
  }
  DynBox_Entries_Col1 {
   widgetClassPath=&DynamicEntryBox
   bIsVariable=true
   parent=HorizontalBox_0
   widget=DynBox_Entries_Col1
   properties {
    entryBoxType=Vertical
    entryWidgetClass=&/WBP_Pilgrim_SortFilterButton.WBP_Pilgrim_SortFilterButton_C
    numDesignerPreviewEntries=5
    entrySpacing=<0 12>
   }
   slot {class=HorizontalBoxSlot, properties {padding=<ltrb 0.4 0.4 0.4 0.4>, size {sizeRule=Fill, value=1}}}
  }
  Spacer_51 {
   widgetClassPath=&Spacer
   parent=HorizontalBox_0
   widget=Spacer_51
   slot {class=HorizontalBoxSlot, properties {size {sizeRule=Fill, value=0.2}}}
  }
  DynBox_Entries_Col2 {
   widgetClassPath=&DynamicEntryBox
   bIsVariable=true
   parent=HorizontalBox_0
   widget=DynBox_Entries_Col2
   properties {
    entryBoxType=Vertical
    entryWidgetClass=&WBP_Pilgrim_SortFilterButton_C
    entrySpacing=<0 12>
   }
   slot {class=HorizontalBoxSlot, properties {padding=<ltrb 0.4 0.4 0.4 0.4>, size {sizeRule=Fill, value=1}}}
  }
 ]
}
